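# Material function for the FeelUnionMoo backpack cosmetic that blends a flipbook mask into the emissive channel of the incoming material attributes.
# UV channel 1, scaled and offset by parameters, feeds a 2x2 FlipBook whose phase is set by the React parameter.
# The sampled red channel serves as both the emissive value and blend alpha, leaving other attributes untouched.
# A constant temporal responsiveness output is also written.
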MaterialFunction /BRCosmetics/Accessories/FORT_Backpacks/Backpack_F_MED_FeelUnionMoo/Materials/MF_FeelUnionMoo
graph {nodes [
  n1 = GetMaterialAttributes(None=FunctionInput() {inputType=FunctionInput_MaterialAttributes}) {attributeGetTypes [#1]}
  n2 = Constant2Vector() {g=1, r=2}
  n3 = Multiply(
   A=TextureCoordinate() {coordinateIndex=1}
   B=AppendVector(A=ScalarParameter() {defaultValue=1, parameterName=SizeX}, B=ScalarParameter() {defaultValue=1, parameterName=SizeY})
  ) {constB=1.25}
  n4 = AppendVector(A=ScalarParameter() {parameterName=OffsetX}, B=ScalarParameter() {parameterName=OffsetY})
  n5 = TextureSample(UVs=MaterialFunctionCall("Animation  Phase (0-1)"=ScalarParameter() {parameterName=React}, Number of Rows=@n2.R, Number of Columns=@n2.R, UVs=Add(A=@n3, B=@n4)) {materialFunction=&FlipBook}.UVs) {samplerType=SAMPLERTYPE_Masks, texture=/BRCosmetics/Accessories/FORT_Backpacks/Backpack_F_MED_FeelUnionMoo/Textures/FeelUnionMoo_flipbook}
  FunctionOutput(None=SetMaterialAttributes!(MaterialAttributes=@n1.MaterialAttributes, EmissiveColor=LinearInterpolate(A=@n1.EmissiveColor, B=@n5.R, Alpha=@n5.R) {constB=0}) {attributeSetTypes [#1]})
  Multiply(A=@n5.A)
  TemporalResponsivenessOutput!(None=Constant() {r=0.4})
 ]}
